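; Looping gameplay cue that renders a cloak effect on a Fort player pawn, with delayed material overrides and Niagara bursts.
; On start it loads balance values, binds interact-search visibility modifiers, waits for cosmetic meshes, then cloaks.
; Tick writes computed visibility to a material parameter collection's Opacity and feeds smoothed speed to looping audio.
; Local, replay and broadcast views schedule a pulsing expiry warning; removal clears timers and uncloaks.
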(event EventOnLoopingStart(EitherVFX-preferthisversion)
  (params (MyTarget :type Actor%20Object%20Reference) (Parameters :type Gameplay%20Cue%20Parameters%20Structure%20%28by%20ref%29) (ParticleComponents :type Array%20of%20FXSystem%20Component%20Object%20References) (AudioComponents :type Array%20of%20Audio%20Component%20Object%20References))
  (bind _targetplayer (Variables|FortGameCueNotifyLoop_Cloak|GetTargetPlayer))
  ;[0 1 0.05708] Looping Start
  (|CallParentFunction MyTarget Parameters ParticleComponents AudioComponents)
  (Utilities|FlowControl|Sequence
    (:then_0
      ;;[0 1 0.09341] Initialze variables
      (bind _asfort_player_pawn (Utilities|Casting|CastToFortPlayerPawn MyTarget)
        (:then
          (Variables|FortGameCueNotifyLoop_Cloak|SetTargetPlayer _asfort_player_pawn))
        (:CastFailed
          _))
      (Actor|Tick|SetActorTickEnabled true)
      (Variables|FortGameCueNotifyLoop_Cloak|SetStationaryVisMult (ScalableFloat|GetValueatLevel (Variables|Balance|GetRow_StationaryVisMult)))
      (Variables|FortGameCueNotifyLoop_Cloak|SetMaxSpeedVisMult (ScalableFloat|GetValueatLevel (Variables|Balance|GetRow_MaxSpeedVisMult)))
      (Variables|FortGameCueNotifyLoop_Cloak|SetVisibilityMinFriendly (ScalableFloat|GetValueatLevel (Variables|Balance|GetRow_FriendlyVisMultMin)))
      (Variables|FortGameCueNotifyLoop_Cloak|SetVisibilityMinNonfriendly (ScalableFloat|GetValueatLevel (Variables|Balance|GetRow_NonfriendlyVisMultMin)))
      (Variables|Balance|CalculatedValues|SetInteractVisMultAddMax (ScalableFloat|GetValueatLevel (Variables|Balance|GetRow_InteractVisMultAddMax)))
      (Variables|FortGameCueNotifyLoop_Cloak|SetVisibilityLevelChangeRate (ScalableFloat|GetValueatLevel (Variables|Balance|GetRow_VisibilityLevelChangeRate))))
    (:then_3
      ;;[0.241 0.4519 1] Hook up callbacks to fade in and out when interacting
      (CallFunction|SetModifierEnabled (Variables|Default|GetInteract%20Modifier%20Name%20Const))
      (EventDispatchers|BindEventtoOnStartedInteractSearch _targetplayer (EventDispatchers|CreateEvent :!event Custom|StartedInteract))
      (EventDispatchers|BindEventtoOnEndedInteractSearch _targetplayer (EventDispatchers|CreateEvent :!event Custom|EndedInteract)))
    (:then_4
      ;;[0 0.9642 1] Set-up for the local player
      (if (CallFunction|ShouldRenderasLocalPlayer)
        ;;;[1 0.2138 0.6332] Handle ability expire warning - local player only
        (if (> (ScalableFloat|GetValueatLevel (Variables|Balance|GetRow_ExpireTellDuration)) 0.0)
          (CallFunction|ClearExpireTell)
          (bind _returnvalue (Utilities|Time|SetTimerbyEvent (EventDispatchers|CreateEvent :!event Custom|BeginExpireTell) (- (ScalableFloat|GetValueatLevel (Variables|Balance|GetRow_Duration)) (ScalableFloat|GetValueatLevel (Variables|Balance|GetRow_ExpireTellDuration)))))
          (Variables|Default|SetExpireTellDelayTimer _returnvalue))))
    (:then_5
      ;;[0.7159 1 0] Initialize the skeletal meshes
      (CallFunction|InitializePlayerSkeletalMeshes))))

(event Custom|Initialize%20Player%20Skeletal%20Meshes
  (bind _targetplayer (Variables|FortGameCueNotifyLoop_Cloak|GetTargetPlayer))
  ;[0.4358 1 0] Initialize Player Skeletal Meshes
  (Utilities|DoOncePerFrame
    (:%20%20
      (Utilities|Array|Clear (Variables|Default|GetPlayerSkeletalMeshes))
      (Utilities|IsValid _targetplayer
        (:Is%20Valid
          (bind _delegatehandlecontroller_out (Cosmetics|WhenCosmeticsApplied _targetplayer (delegate Custom|OnCosmeticsApplied)))
          (Variables|Default|SetDelegate%20Handle%20Controller _delegatehandlecontroller_out))))))

(event Custom|OnCosmeticsApplied
  (params (Pawn :type Fort%20Player%20Pawn%20Object%20Reference) (BaseMesh :type Skeletal%20Mesh%20Component%20Object%20Reference) (PartsArray :type Array%20of%20Custom%20Character%20Part%20Object%20References) (ValidSkeletalMeshComponentPartArray :type Array%20of%20Skeletal%20Mesh%20Component%20Object%20References))
  (bind _playerskeletalmeshes (Variables|Default|GetPlayerSkeletalMeshes))
  ;[0.4358 1 0] Initialize Player Skeletal Meshes
  ;;[0.1083 1 0.0636] OnCosmeticsApplied: Collect updated meshes and try to apply the cloak material to them.
  (Utilities|FlowControl|Sequence
    (:then_0
      ;;;[0.02461 0 1] Populate PlayerSkeletalMeshes
      (Utilities|Array|Clear _playerskeletalmeshes)
      (for _array_element PartsArray
        (Utilities|IsValid _array_element
          (:Is%20Valid
            (Utilities|Array|Add _playerskeletalmeshes)))))
    (:then_1
      ;;;[1 0.5431 0] If we do NOT have skeletal meshes yet, then schedule a retry.
      ;;; If we do have skeletal meshes then set those meshes to use the cloaked material.
      (if (<= (Utilities|Array|Length _playerskeletalmeshes) 0)
        (Utilities|Time|SetTimerforNextTickbyEvent (EventDispatchers|CreateEvent :!event Custom|Initialize%20Player%20Skeletal%20Meshes))
        (else
          (CallFunction|SetMaterialCloaked))))))

(event EventOnRemoval(EitherVFX-preferthisversion)
  (params (MyTarget :type Actor%20Object%20Reference) (Parameters :type Gameplay%20Cue%20Parameters%20Structure%20%28by%20ref%29) (ParticleComponents :type Array%20of%20FXSystem%20Component%20Object%20References) (AudioComponents :type Array%20of%20Audio%20Component%20Object%20References) (BurstCameraShakeInstance :type Camera%20Shake%20Base%20Object%20Reference) (BurstDecalInstance :type Decal%20Actor%20Object%20Reference))
  (bind _targetplayer (Variables|FortGameCueNotifyLoop_Cloak|GetTargetPlayer))
  ;[1 0.4543 0] Handle Removal
  (|CallParentFunction MyTarget Parameters ParticleComponents AudioComponents BurstCameraShakeInstance BurstDecalInstance)
  (bind _asfort_player_pawn (Utilities|Casting|CastToFortPlayerPawn MyTarget)
    (:then
      (Variables|FortGameCueNotifyLoop_Cloak|SetTargetPlayer _asfort_player_pawn))
    (:CastFailed
      _))
  (Actor|Tick|SetActorTickEnabled false)
  (CallFunction|ClearExpireTell)
  (Utilities|IsValid _targetplayer
    (:Is%20Valid
      (EventDispatchers|UnbindEventfromOnStartedInteractSearch _targetplayer (EventDispatchers|CreateEvent :!event Custom|StartedInteract))
      (EventDispatchers|UnbindEventfromOnStartedInteractSearch _targetplayer (EventDispatchers|CreateEvent :!event Custom|EndedInteract))
      (CallFunction|SetMaterialUncloaked))))

(event EventTick
  (params (DeltaSeconds :type Float%20%28single-precision%29))
  ;[0.07843 0.4235 1] Event tick
  (Variables|FortGameCueNotifyLoop_Cloak|GetTargetPlayer
    (:then
      (Utilities|FlowControl|Sequence
        (:then_0
          ;;[0 1 1] Checks for what the player is doing. Determines the material opacity
          (bind _returnvalue (CallFunction|TickVisibilityLevel DeltaSeconds))
          (Rendering|Material|SetScalarParameterValue 0 "/CloakGameplay/Materials/MPC_Cloak_Generic.MPC_Cloak_Generic" "Opacity" _returnvalue :!class /Script/Engine.KismetMaterialLibrary)
          (if (Variables|Default|GetDebugLogging%20Const)
            (Development|PrintString 0 (Utilities|String|Append "GCNL_Cloak_Generic_Loop::Tick() TargetPlayer: [" (Fortnite|GetPlayerNameSafe (Variables|FortGameCueNotifyLoop_Cloak|GetTargetPlayer)) "], Current Opacity: [" (Utilities|String|ToString(Float) _returnvalue) "].") true true "(R=0.000000,G=0.660000,B=1.000000,A=1.000000)" 0.0)))
        (:then_1
          ;;[0.05415 1 0.7377] Update Team Visibility
          (CallFunction|UpdateTeamVisbility))
        (:then_2
          ;;[0.6823 0.3646 1] Perform the Audio Tick
          (if (> (Utilities|Array|Length (Variables|Default|GetLoopingAudioComponents)) 0)
            (CallFunction|AudioTick (Math|Vector|VectorLength (Transformation|GetVelocity (Variables|FortGameCueNotifyLoop_Cloak|GetTargetPlayer))))))))
    (:else)))

(event Custom|Audio%20Tick
  (params (Velocity :type Float%20%28double-precision%29))
  (bind _loopingaudiocomponents (Variables|Default|GetLoopingAudioComponents))
  (bind _output (Utilities|Array|Get(acopy) _loopingaudiocomponents))
  ;[0.2431 0.2588 1] Audio Tick
  (if (Utilities|Array|IsValidIndex _loopingaudiocomponents)
    (Utilities|IsValid _output
      (:Is%20Valid
        (bind _output_get (Variables|Balance|CalculatedValues|SetAudioVelocityInterp (Math|Interpolation|FInterpTo (Variables|Balance|CalculatedValues|GetAudioVelocityInterp) (Math|Float|MapRangeClamped Velocity 6e+02 1e+03 0.0 1.0) (Utilities|Time|GetWorldDeltaSeconds 0) 1.5)))
        (Audio|Parameter|SetFloatParameter _output (Variables|ParameterName|GetAudioParameterName_Velocity) _output_get :!class /Script/AudioExtensions.AudioParameterControllerInterface)))))

(event Custom|StartedInteract
  ;[0.2314 0.2118 0.4824] Interact (increase opacity while interacting)
  (CallFunction|SetModifierEnabled (Variables|Default|GetInteract%20Modifier%20Name%20Const) true))

(event Custom|EndedInteract
  ;[0.2314 0.2118 0.4824] Interact (increase opacity while interacting)
  (CallFunction|SetModifierEnabled (Variables|Default|GetInteract%20Modifier%20Name%20Const)))

(event Custom|UpdateTeamVisbility)

(event Custom|BeginExpireTell
  ;[0.543 0 1] Expiration Warning Sequence
  (bind _returnvalue (Utilities|Time|SetTimerbyEvent (delegate Custom|ExpireWarningTick) (Variables|Balance|GetExpirationSoundPeriod) true))
  (Variables|Default|SetExpirationSoundTimer _returnvalue))

(event Custom|ExpireWarningTick
  (bind _expirationsoundperiodscale (Variables|Balance|GetExpirationSoundPeriodScale))
  (bind _expirationsoundperiod (Variables|Balance|GetExpirationSoundPeriod))
  (bind _returnvalue (* _expirationsoundperiod _expirationsoundperiodscale))
  (bind _returnvalue_2 (/ _returnvalue 2.0))
  (bind _returnvalue_3 (Utilities|Time|GetGameTimeinSeconds 0))
  (bind _targetplayer (Variables|FortGameCueNotifyLoop_Cloak|GetTargetPlayer))
  ;[0.543 0 1] Expiration Warning Sequence
  (Audio|SpawnSoundAttached (Variables|Default|GetEnd%20Warning%20Sound%20Cue) (Variables|/Script/Engine.Actor|GetRootComponent (Variables|FortGameCueNotifyLoop_Cloak|GetTargetPlayer)))
  (Class|FortPlayerPawn|SetScalarParamonOverriddenMaterials _targetplayer "ExpireTimeStampA" _returnvalue_3)
  (Class|FortPlayerPawn|SetScalarParamonOverriddenMaterials _targetplayer "ExpireTimeStampB" (+ _returnvalue_3 _returnvalue_2))
  (Class|FortPlayerPawn|SetScalarParamonOverriddenMaterials _targetplayer "ExpireTransitionLength" _returnvalue_2))

(event Custom|ClearExpireTell
  ;[0.543 0 1] Expiration Warning Sequence
  (Utilities|Time|ClearandInvalidateTimerbyHandle 0 (Variables|Default|GetExpireTellDelayTimer))
  (Utilities|Time|ClearandInvalidateTimerbyHandle 0 (Variables|Default|GetExpirationSoundTimer)))

(event Custom|SetMaterialCloaked
  ;[1 1 1] Cloak Activate
  (Utilities|Time|ClearandInvalidateTimerbyHandle 0 (Variables|FX|GetRemoveCloakMaterialOverrideHandle))
  (bind _targetplayer (Variables|FortGameCueNotifyLoop_Cloak|GetTargetPlayer)
    (:then
      (bind _returnvalue (Niagara|SpawnSystemAttached (Variables|FX|GetActivateFXTemplate) (Variables|/Script/Engine.Character|GetMesh _targetplayer) "root" "0, 0, 0" "0, 0, 0" "KeepRelativeOffset" true true "AutoRelease"))
      (bind _output_get (Variables|Default|SetActivateCloakFX _returnvalue))
      (Effects|Components|ParticleSystem|SetActorParameter _returnvalue "SkelMesh" (Variables|FortGameCueNotifyLoop_Cloak|GetTargetPlayer))
      (bind _returnvalue_2 (Utilities|Time|SetTimerbyFunctionName 0 "ApplyCloakMaterialOverride" (Variables|FX|GetDelayBeforeApplyingCloakMaterialOverride)))
      (Variables|FX|SetApplyCloakMaterialOverrideHandle _returnvalue_2))
    (:else)))

(event Custom|SetMaterialUncloaked
  (bind _targetplayer (Variables|FortGameCueNotifyLoop_Cloak|GetTargetPlayer))
  ;[1 1 1] Cloak Remove
  (Utilities|Time|ClearandInvalidateTimerbyHandle 0 (Variables|FX|GetApplyCloakMaterialOverrideHandle))
  (Class|FortPlayerPawn|SetScalarParamonOverriddenMaterials _targetplayer "TimeStampA" 0.0)
  (Class|FortPlayerPawn|SetScalarParamonOverriddenMaterials _targetplayer "TimeStampB" (Utilities|Time|GetGameTimeinSeconds 0))
  (bind _returnvalue (Utilities|Time|SetTimerbyFunctionName 0 "RemoveCloakMaterialOverride" (Variables|FX|GetDelayBeforeRemovingCloakMaterialOverride)))
  (Variables|FX|SetRemoveCloakMaterialOverrideHandle _returnvalue))

(fn ApplyCloakMaterialOverride
  (bind _targetplayer (Variables|FortGameCueNotifyLoop_Cloak|GetTargetPlayer))
  (Rendering|SetCastShadow (Variables|/Script/Engine.Character|GetMesh (Variables|FortGameCueNotifyLoop_Cloak|GetTargetPlayer)) false)
  (bind _returnvalue (Class|FortPlayerPawn|ApplyMaterialOverrideLocalwithFlags (Variables|FortGameCueNotifyLoop_Cloak|GetTargetPlayer) (Utilities|ResolveSoftReference (Variables|FX|GetM_Cloaked)) (Variables|Default|GetMaterialParametersToCopy) 0.0 35))
  (Variables|Default|SetMaterialOverrideId _returnvalue)
  (Class|FortPlayerPawn|SetScalarParamonOverriddenMaterials _targetplayer "TimeStampA" (Utilities|Time|GetGameTimeinSeconds 0))
  (Class|FortPlayerPawn|SetScalarParamonOverriddenMaterials _targetplayer "TimeStampB" 1e+10))

(fn IsPlayerControllerFriendly
  (params (PlayerController :type Controller%20Object%20Reference))
  (Utilities|IsValid PlayerController
    (:Is%20Valid
      (bind _targetplayer (Variables|FortGameCueNotifyLoop_Cloak|GetTargetPlayer)
        (:then
          ;[1 0.4635 0.4635] Both players are valid, return if they are on the same team
          (return (Team|OnSameTeam (Variables|/Script/Engine.Pawn|GetPlayerState _targetplayer) (Variables|/Script/Engine.Controller|GetPlayerState PlayerController))))
        (:else
          _)))
    (:Is%20Not%20Valid
      _))
  ;[1 0.4635 0.4635] One or both players is nullptr, return FALSE
  (return false))

(fn RemoveCloakMaterialOverride
  (Utilities|FlowControl|Sequence
    (:then_0
      (Class|FortPlayerPawn|RemoveMaterialOverrideLocal (Variables|FortGameCueNotifyLoop_Cloak|GetTargetPlayer) (Variables|Default|GetMaterialOverrideId))
      (Guid|InvalidateGuid (Variables|Default|GetMaterialOverrideId))
      (bind _targetplayer (Variables|FortGameCueNotifyLoop_Cloak|GetTargetPlayer)
        (:then
          (Niagara|SpawnSystemAttached (Variables|FX|GetDeactivateFXTemplate) (Variables|/Script/Engine.Character|GetMesh _targetplayer) "root" "0, 0, 0" "0, 0, 0" "KeepRelativeOffset" true true "AutoRelease"))
        (:else)))
    (:then_1
      (Rendering|SetCastShadow (Variables|/Script/Engine.Character|GetMesh (Variables|FortGameCueNotifyLoop_Cloak|GetTargetPlayer)) true)
      (Variables|FortGameCueNotifyLoop_Cloak|SetTargetPlayer 0))))

(fn ShouldRenderAsLocalPlayer
  (bind _targetplayer (Variables|FortGameCueNotifyLoop_Cloak|GetTargetPlayer)
    (:then
      ;[1 0.4896 0.4896] Return if this GC shourender as if this is the local player
      (return (Math|Boolean|ORBoolean (Pawn|IsLocallyControlled _targetplayer) (Replay|GetIsPlayingReplay 0) (Broadcast|GetIsBroadcastClient 0))))
    (:else
      ;[1 0.4896 0.4896] Could not find Target Player, Return FALSE
      (return false))))

(fn ConstructionScript)
